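# Niagara idle visualizer for the FareSpore keytar cosmetic, drawing an audio-reactive equalizer from mesh and sprite chips.
# Chips burst-spawns a persistent columns-by-rows grid, colored along each column by a gradient.
# A local MultsFromSpectrum module turns music spectrum, amplitude and waveform noise into mesh scale and per-row sprite cutoff.
# Preview mode substitutes a fake spectrum and pure noise; the Chips001, ChipsNorm and Values emitters are disabled.
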NiagaraSystem /SparksCosmetics/Instrument/Keytar/FareSpore/Effects/NS_Keytar_FareSpore_Idle_Visualizer_v1
dependencies {
 usedDataInterfaces [&NiagaraDataInterfaceColorCurve, &NiagaraDataInterfaceCurve]
 usedDynamicInputs [&MakeLinearColor, &MakeCustomLinearColorFromBool]
 usedModules [
  &self:SystemSpawnScript.SystemScriptSource.SystemScriptGraph.NiagaraNodeAssignment_4.SetVariables_#1
  &self:SystemSpawnScript.SystemScriptSource.SystemScriptGraph.NiagaraNodeAssignment_5.SetVariables_#2
  &SystemState
  &self:SystemSpawnScript.SystemScriptSource.SystemScriptGraph.NiagaraNodeAssignment_1.SetVariables_#3
  &self:SystemSpawnScript.SystemScriptSource.SystemScriptGraph.NiagaraNodeAssignment_0.SetVariables_#4
  &self:SystemSpawnScript.SystemScriptSource.SystemScriptGraph.NiagaraNodeAssignment_2.SetVariables_#5
  &self:SystemSpawnScript.SystemScriptSource.SystemScriptGraph.NiagaraNodeAssignment_9.SetVariables_#6
  &self:SystemSpawnScript.SystemScriptSource.SystemScriptGraph.NiagaraNodeAssignment_3.SetVariables_#7
 ]
}
emitters {
 Chips {
  emitterUpdateScript [
   &EmitterState()
   &SpawnBurst_Instantaneous(
    Spawn Count=&Multiply_Int(A=$System.Columns, B=$System.Rows)
    Use Spawn Probability=false
    UseLoopCountLimit=true
   )
  ]
  particleSpawnScript [
   &PartitionParticles(
    Normalized Index In Partition Includes One=true
    Number Of Partitions=$System.Columns
    Particles Per Partition:NiagaraInt32=$System.Rows
    Partition Mode:ENiagara_PartitionMode_v2=::NewEnumerator1
   )
   &InitializeParticle(
    Color Mode=::NewEnumerator0
    Mesh Scale=&Multiply_VectorByFloat(Vector=<0.1 0.02 0.02>, Float=0.28)
    Mesh Scale Mode=::NewEnumerator4
    Position Offset=&Add_Vector(
     A=<16 0 10>
     B=&Add_Vector(
      A=&Lerp_Vector(
       A=<0 0 -16.7>
       B=<0 0 16.7>
       Alpha=&Multiply_Float(A=$Particles.PartitionParticles.NormalizedPartition, B=1.1667)
      )
      B=&Lerp_Vector(A=<0 0 0>, B=<15 0 0>, Alpha=$Particles.PartitionParticles.Normalized Index In Partition)
     )
    )
    Sprite Rotation Angle=90
    Sprite Rotation Mode=::NewEnumerator0
    Sprite Size Mode=::NewEnumerator3
    Uniform Sprite Size=5
    UsePositionOffset=true
    Write Material Random=true
   )
  ]
  particleUpdateScript [
   &ParticleState(Kill Particles When Lifetime Has Elapsed=false)
   /SetVariables(Transient.NoiseCurve:NiagaraFloat=&Waveform(
     Waveform Count=::NewEnumerator1
     "Waveform [1]"=::NewEnumerator0
     "[1] Phase Scale"=&Multiply_FloatByInt(Float=0.2, Integer=$Particles.PartitionParticles.Partition)
     "[1] Frequency"=1.7
     "[1] Amplitude Min/Max"=<0 1>
     "[2] Average"=false
     "[2] Input Position"=0
     "[2] Phase"=$Emitter.Age
     "[2] Phase Scale"=&Multiply_FloatByInt(
      Float=0.2
      Integer=/Niagara/DynamicInputs/Subtract/Subtract_Integer(A:NiagaraInt32=$System.Columns, B:NiagaraInt32=$Particles.PartitionParticles.Partition)
     )
     "[2] Compound Frequency"=2
     "[2] Frequency"=1.5
     "[2] Amplitude Min/Max"=<0 1>
     "[2] Scale"=1
     "[2] Bias"=0
     "[2] Exponent"=1
     "[2] Pulse Duration"=0.1
     "[2] Clip Triangle Min/Max"=<0 0>
     "[2] Random Seed"=0
    )) {isSetParameters=true, module=SetVariables_#8}
   /SetVariables(Transient.EQCurve:LinearColor=&ColorFromCurve(
     ColorCurve=DataInterface() {propertyValues {
       AlphaCurve {keys [<at 0 1 RCIM_Cubic>]}
       BlueCurve {keys [<at 0.05 1>,<at 0.3456 1 RCIM_Cubic>,<at 0.8091 1 RCIM_Cubic>]}
       GreenCurve {keys [<at 0.05 0.6344>, {arriveTangent=-0.8357, interpMode=RCIM_Cubic, leaveTangent=-0.8357, time=0.3456, value=0.8218}, <at 0.8091 0 RCIM_Cubic>]}
       RedCurve {keys [<at 0.05 0.04>, {arriveTangent=0.8087, interpMode=RCIM_Cubic, leaveTangent=0.8087, time=0.3456, value=0.44}, <at 0.8091 0.6539 RCIM_Cubic>]}
      }}
     CurveIndex=$Particles.PartitionParticles.Normalized Index In Partition
    )) {isSetParameters=true, module=SetVariables_#9}
   /SetVariables(Transient.EQCurve:LinearColor=&ColorFromCurve(
     ColorCurve=DataInterface() {propertyValues {
       AlphaCurve {keys [<at 0 1 RCIM_Cubic>]}
       BlueCurve {keys [
         <at 0 1>
         {arriveTangent=-0.2516, interpMode=RCIM_Cubic, leaveTangent=-0.2516, time=0.5713, value=0.7294}
         <at 0.592 0.851>
         <at 0.8091 1 RCIM_Cubic>
        ]}
       GreenCurve {keys [
         <at 0 0.327>
         {arriveTangent=0.8852, interpMode=RCIM_Cubic, leaveTangent=0.8852, time=0.5713, value=1}
         <at 0.592 0.851>
         <at 0.8091 0 RCIM_Cubic>
        ]}
       RedCurve {keys [
         <at 0 0>
         {arriveTangent=1.6891, interpMode=RCIM_Cubic, leaveTangent=1.6891, time=0.5713, value=0.226}
         <at 0.592 1>
         <at 0.8091 0.6539 RCIM_Cubic>
        ]}
      }}
     CurveIndex=$Particles.PartitionParticles.Normalized Index In Partition
    )) {enabled=false, isSetParameters=true, module=SetVariables_#10}
   &self:MultsFromSpectrum(
    Column Selector:NiagaraInt32=$Particles.PartitionParticles.Partition
    Modifier:NiagaraFloat=&Lerp_Float(
     A=&Multiply_Float(
      A=1.5
      B=&FloatFromCurve(
       FloatCurve=DataInterface() {propertyValues {Curve {keys [<at 0 0.75>,<at 0.25 1>,<at 1 0.5>]}}}
       CurveIndex=&Multiply_FloatByInt(Float=0.125, Integer=$Particles.PartitionParticles.Partition)
       Scale Curve=$NPC.NPC_Audio_Reactive_Parameters.Amplitude_Music
      )
     )
     B=$Transient.NoiseCurve
     Alpha=&MakeCustomFloatFromBool(Bool=$System.bIsPreview, True Float=1, False Float=0.25)
    )
    Power Spectrum=3
    Row Selector:NiagaraFloat=$Particles.PartitionParticles.Normalized Index In Partition
    Spectrum Param:LinearColor=&LerpLinearColors(
     StartColor=$NPC.NPC_Audio_Reactive_Parameters.Spectrum_Music
     EndColor=$System.FakeSpectrum
     LerpFactor=&MakeCustomFloatFromBool(Bool=$System.bIsPreview, True Float=1, False Float=0)
    )
   ) {module=MultsFromSpectrum}
   &ScaleColor(
    Scale Alpha=&FloatFromCurve(
     FloatCurve=DataInterface() {propertyValues {Curve {keys [<at 0 1>,<at 1 0.5>]}}}
     CurveIndex=$Particles.PartitionParticles.Normalized Index In Partition
    )
    Scale RGB=&MakeVectorFromLinearColorRGB(Color=$Transient.EQCurve)
    ScaleA=true
    ScaleRGB=true
   )
   &ScaleMeshSize(Scale Factor=&VectorFromFloat(Value=$Output.MultsFromSpectrum.SoftMult))
   &ScaleSpriteSize(
    Scale Sprite Size Mode=::NewEnumerator0
    Uniform Scale Factor=&MakeCustomFloatFromBool(
     Bool=&SetBoolByFloatComparison(A=$Output.MultsFromSpectrum.ModifiedMult, B=$Particles.PartitionParticles.Normalized Index In Partition, Comparison Type=::NewEnumerator6)
     True Float=0
     False Float=1
    )
   )
   /SetVariables(
    Particles.SpriteColor=&MakeLinearColorFromVectorAndFloat(
     "Vector (RGB)"=&MakeVectorFromLinearColorRGB(Color=$Particles.Color)
     "Float (Alpha)"=&FloatFromCurve(
      FloatCurve=DataInterface() {propertyValues {Curve {keys [<at 0 0.2>,<at 1 0.4>]}}}
      CurveIndex=$Particles.PartitionParticles.Normalized Index In Partition
      Scale Curve=&MakeCustomFloatFromBool(
       Bool=&SetBoolByFloatComparison(A=$Output.MultsFromSpectrum.ModifiedMult, B=$Particles.PartitionParticles.Normalized Index In Partition, Comparison Type=::NewEnumerator6)
       True Float=0
       False Float=1
      )
     )
    )
    Particles.SpritePosition:NiagaraPosition=&AddVectorToPosition(Position=$Particles.Position, Vector=<1 0 0>)
   ) {isSetParameters=true, module=SetVariables_#11}
   /SetVariables(
    Particles.SpriteColor=&MakeLinearColorFromVectorAndFloat(
     "Vector (RGB)"=&MakeVectorFromLinearColorRGB(Color=&MultiplyLinearColor(Linear Color=$Particles.Color, Scale Factor=&MakeVector4_FromLinearColor(LinearColor=$Particles.Color)))
     "Float (Alpha)"=&FloatFromCurve(
      FloatCurve=DataInterface() {propertyValues {Curve {keys [<at 0 0.2>,<at 1 0.4>]}}}
      CurveIndex=$Particles.PartitionParticles.Normalized Index In Partition
      Scale Curve=&MakeCustomFloatFromBool(
       Bool=&SetBoolByFloatComparison(A=$Output.MultsFromSpectrum.ModifiedMult, B=$Particles.PartitionParticles.Normalized Index In Partition, Comparison Type=::NewEnumerator6)
       True Float=0
       False Float=1
      )
     )
    )
    Particles.SpritePosition:NiagaraPosition=&AddVectorToPosition(Position=$Particles.Position, Vector=<1 0 0>)
   ) {isSetParameters=true}
   /SetVariables(
    Particles.SpriteColor=&MakeLinearColorFromVectorAndFloat(
     "Vector (RGB)"=&MakeVectorFromLinearColorRGB(Color=/Niagara/DynamicInputs/Subtract/Subtract_LinearColor(
       A:LinearColor=<rgba 1 1 1 0>
       B:LinearColor=&HueShiftLinearColor(Hue=0.5, Saturation=1, Value=1, Linear Color=$Particles.Color)
      ))
     "Float (Alpha)"=&FloatFromCurve(
      FloatCurve=DataInterface() {propertyValues {Curve {keys [<at 0 0.2>,<at 1 0.4>]}}}
      CurveIndex=$Particles.PartitionParticles.Normalized Index In Partition
      Scale Curve=&MakeCustomFloatFromBool(
       Bool=&SetBoolByFloatComparison(A=$Output.MultsFromSpectrum.ModifiedMult, B=$Particles.PartitionParticles.Normalized Index In Partition, Comparison Type=::NewEnumerator6)
       True Float=0
       False Float=1
      )
     )
    )
    Particles.SpritePosition:NiagaraPosition=&AddVectorToPosition(Position=$Particles.Position, Vector=<1 0 0>)
   ) {enabled=false, isSetParameters=true}
  ]
  rendererClasses [&NiagaraSpriteRendererProperties, &NiagaraMeshRendererProperties]
  renderers [
   {rendererClass=&NiagaraSpriteRendererProperties}
   {rendererClass=&NiagaraMeshRendererProperties, rendererIndex=1}
  ]
 }
 Chips001 {
  bEnabled=false
  emitterUpdateScript [
   &EmitterState()
   &SpawnBurst_Instantaneous(
    Spawn Count=&Multiply_Int(A=$System.Columns, B=$System.Rows)
    Use Spawn Probability=false
    UseLoopCountLimit=true
   )
  ]
  particleSpawnScript [
   &PartitionParticles(
    Normalized Index In Partition Includes One=true
    Number Of Partitions=$System.Columns
    Particles Per Partition:NiagaraInt32=$System.Rows
    Partition Mode:ENiagara_PartitionMode_v2=::NewEnumerator1
   )
   &InitializeParticle(
    Color Mode=::NewEnumerator0
    Mesh Scale=&Multiply_VectorByFloat(Vector=<0.03 0.03 0.1>, Float=0.28)
    Mesh Scale Mode=::NewEnumerator4
    Position Offset=&Add_Vector(
     A=<8.6 -2.5 -9.75>
     B=&Add_Vector(
      A=&Lerp_Vector(
       A=<0 0 -11.15>
       B=<0 0 11.15>
       Alpha=&Multiply_Float(A=$Particles.PartitionParticles.NormalizedPartition, B=1.1667)
      )
      B=&Lerp_Vector(A=<0 0 0>, B=<15 0 0>, Alpha=$Particles.PartitionParticles.Normalized Index In Partition)
     )
    )
    Sprite Size Mode=::NewEnumerator3
    Uniform Sprite Size=5
    UsePositionOffset=true
    Write Material Random=true
   )
  ]
  particleUpdateScript [
   &ParticleState(Kill Particles When Lifetime Has Elapsed=false)
   /SetVariables(Transient.NoiseCurve:NiagaraFloat=&MakeCustomFloatFromBool(
     Bool=$System.bIsPreview
     True Float=&Waveform(
      Waveform Count=::NewEnumerator1
      "Waveform [1]"=::NewEnumerator0
      "[1] Phase Scale"=&Multiply_FloatByInt(Float=0.2, Integer=$Particles.PartitionParticles.Partition)
      "[1] Frequency"=1.7
      "[1] Amplitude Min/Max"=<0 1>
      "[2] Average"=false
      "[2] Input Position"=0
      "[2] Phase"=$Emitter.Age
      "[2] Phase Scale"=&Multiply_FloatByInt(
       Float=0.2
       Integer=&Subtract_Integer(A:NiagaraInt32=$System.Columns, B:NiagaraInt32=$Particles.PartitionParticles.Partition)
      )
      "[2] Compound Frequency"=2
      "[2] Frequency"=1.5
      "[2] Amplitude Min/Max"=<0 1>
      "[2] Scale"=1
      "[2] Bias"=0
      "[2] Exponent"=1
      "[2] Pulse Duration"=0.1
      "[2] Clip Triangle Min/Max"=<0 0>
      "[2] Random Seed"=0
     )
     False Float=$NPC.NPC_Audio_Reactive_Parameters.BeatGameTime
    )) {isSetParameters=true}
   /SetVariables(Transient.NoiseCurve:NiagaraFloat=&MakeCustomFloatFromBool(
     Bool=$System.bIsPreview
     True Float=&Sine(Normalized Angle=$Emitter.Age, Period=0.6, Scale=0.5, Bias=0.5)
     False Float=$NPC.NPC_Audio_Reactive_Parameters.BeatGameTime
    )) {enabled=false, isSetParameters=true}
   /SetVariables(Transient.EQCurve:LinearColor=&ColorFromCurve(
     ColorCurve=DataInterface() {propertyValues {
       AlphaCurve {keys [<at 0 1 RCIM_Cubic>]}
       BlueCurve {keys [
         <at 0 1>
         {arriveTangent=-0.2516, interpMode=RCIM_Cubic, leaveTangent=-0.2516, time=0.5713, value=0.7294}
         <at 0.592 0.851>
         <at 0.8091 1 RCIM_Cubic>
        ]}
       GreenCurve {keys [
         <at 0 0.327>
         {arriveTangent=0.8852, interpMode=RCIM_Cubic, leaveTangent=0.8852, time=0.5713, value=1}
         <at 0.592 0.851>
         <at 0.8091 0 RCIM_Cubic>
        ]}
       RedCurve {keys [
         <at 0 0>
         {arriveTangent=1.6891, interpMode=RCIM_Cubic, leaveTangent=1.6891, time=0.5713, value=0.226}
         <at 0.592 1>
         <at 0.8091 0.6539 RCIM_Cubic>
        ]}
      }}
     CurveIndex=$Particles.PartitionParticles.Normalized Index In Partition
    )) {isSetParameters=true, module=SetVariables_#9}
   /SetVariables(Transient.EQCurve:LinearColor=&ColorFromCurve(
     ColorCurve=DataInterface() {propertyValues {
       AlphaCurve {keys [<at 0 1 RCIM_Cubic>]}
       BlueCurve {keys [
         <at 0 1>
         {arriveTangent=-1.6891, interpMode=RCIM_Cubic, leaveTangent=-1.6891, time=0.5713, value=0.5252}
         <at 0.592 0>
         <at 0.8091 0 RCIM_Cubic>
        ]}
       GreenCurve {keys [
         <at 0 0>
         {arriveTangent=1.1633, interpMode=RCIM_Cubic, leaveTangent=1.1633, time=0.5713, value=1}
         <at 0.592 0.6887>
         <at 0.8091 0.0028 RCIM_Cubic>
        ]}
       RedCurve {keys [<at 0 0>, {arriveTangent=1.6891, interpMode=RCIM_Cubic, leaveTangent=1.6891, time=0.5713, value=0}, <at 0.592 1>, <at 0.8091 1 RCIM_Cubic>]}
      }}
     CurveIndex=$Particles.PartitionParticles.Normalized Index In Partition
    )) {enabled=false, isSetParameters=true}
   &self:MultsFromSpectrum(
    Column Selector:NiagaraInt32=$Particles.PartitionParticles.Partition
    Modifier:NiagaraFloat=$Transient.NoiseCurve
    Row Selector:NiagaraFloat=$Particles.PartitionParticles.Normalized Index In Partition
    Spectrum Param:LinearColor=$System.ChooseSpectrum
   ) {module=MultsFromSpectrum}
   &self:MultsFromSpectrum(
    Column Selector:NiagaraInt32=$Particles.PartitionParticles.Partition
    Modifier:NiagaraFloat=&Lerp_Float(A=$NPC.NPC_Audio_Reactive_Parameters.Amplitude_Music, B=$Transient.NoiseCurve, Alpha=1)
    Power Spectrum=2
    Row Selector:NiagaraFloat=$Particles.PartitionParticles.Normalized Index In Partition
    Spectrum Param:LinearColor=&LerpLinearColors(StartColor=$NPC.NPC_Audio_Reactive_Parameters.Spectrum_Music, EndColor=$System.FakeSpectrum, LerpFactor=1)
   ) {module=MultsFromSpectrum003}
   &self:MultsFromSpectrum(
    Column Selector:NiagaraInt32=$Particles.PartitionParticles.Partition
    Modifier:NiagaraFloat=&Power(a=&Lerp_Float(A=$NPC.NPC_Audio_Reactive_Parameters.Amplitude_Music, B=$Transient.NoiseCurve, Alpha=0.25), b=2)
    Row Selector:NiagaraFloat=$Particles.PartitionParticles.Normalized Index In Partition
    Spectrum Param:LinearColor=&LerpLinearColors(StartColor=$NPC.NPC_Audio_Reactive_Parameters.Spectrum_Music, EndColor=$System.FakeSpectrum, LerpFactor=0.25)
   ) {module=MultsFromSpectrum002}
   &self:MultsFromSpectrum(
    Column Selector:NiagaraInt32=$Particles.PartitionParticles.Partition
    Modifier:NiagaraFloat=&MakeCustomFloatFromBool(
     Bool=$System.bIsPreview
     True Float=$Transient.NoiseCurve
     False Float=&Multiply_Float(A=0.5, B=&Add_Float(A=$Transient.NoiseCurve, B=$NPC.NPC_Audio_Reactive_Parameters.Amplitude_Music))
    )
    Row Selector:NiagaraFloat=$Particles.PartitionParticles.Normalized Index In Partition
    Spectrum Param:LinearColor=$System.FakeSpectrum
   ) {module=MultsFromSpectrum001}
   /SetVariables(Transient.AlphaMask:NiagaraFloat=&Add_Float(
     A=$Transient.NoiseCurve
     B=&Multiply_Float(A=0.5, B=&Power(a=$Output.MultsFromSpectrum.BaseMult, b=2))
    )) {isSetParameters=true}
   /SetVariables(Transient.AlphaMask:NiagaraFloat=&Power(
     a=&Add_Float(
      A=$Transient.NoiseCurve
      B=&Multiply_Float(
       A=0.5
       B=&Add_Float(
        A=&Power(a=$NPC.NPC_Audio_Reactive_Parameters.Amplitude_Music, b=$NPC.NPC_Audio_Reactive_Parameters.Amplitude_Music)
        B=&Power(a=$Output.MultsFromSpectrum.BaseMult, b=2)
       )
      )
     )
     b=2
    )) {enabled=false, isSetParameters=true}
   /SetVariables(Transient.AlphaMask:NiagaraFloat=&Multiply_Float(
     A=0.5
     B=&Add_Float(
      A=&Power(a=$NPC.NPC_Audio_Reactive_Parameters.Amplitude_Music, b=$NPC.NPC_Audio_Reactive_Parameters.Amplitude_Music)
      B=&Power(a=$Output.MultsFromSpectrum.BaseMult, b=2)
     )
    )) {enabled=false, isSetParameters=true}
   /SetVariables(Transient.AlphaMask:NiagaraFloat=&Power(a=$Output.MultsFromSpectrum.BaseMult, b=2)) {enabled=false, isSetParameters=true}
   &ScaleColor(
    Scale Alpha=&MakeCustomFloatFromBool(
     Bool=&SetBoolByFloatComparison(A=$Transient.AlphaMask, B=$Particles.PartitionParticles.Normalized Index In Partition, Comparison Type=::NewEnumerator6)
     True Float=0
     False Float=1
    )
    Scale RGB=&MakeVectorFromLinearColorRGB(Color=$Transient.EQCurve)
    ScaleA=false
    ScaleRGB=true
   )
   &ScaleMeshSize(Scale Factor=&MakeVector(X=$Output.MultsFromSpectrum.SoftMult, Y=$Output.MultsFromSpectrum.SoftMult, Z=1))
   /SetVariables(Particles.SpriteColor=&Multiply_LinearColorByFloat(Linear Color=$Particles.Color, Float=0.25)) {isSetParameters=true, module=SetVariables_#11}
  ]
  rendererClasses [&NiagaraSpriteRendererProperties, &NiagaraMeshRendererProperties]
  renderers [
   {rendererClass=&NiagaraSpriteRendererProperties}
   {rendererClass=&NiagaraMeshRendererProperties, rendererIndex=1}
  ]
 }
 ChipsNorm {
  bEnabled=false
  emitterUpdateScript [
   &EmitterState()
   &SpawnBurst_Instantaneous(
    Spawn Count=&Multiply_Int(A=$System.Columns, B=$System.Rows)
    Use Spawn Probability=false
    UseLoopCountLimit=true
   )
  ]
  particleSpawnScript [
   &PartitionParticles(
    Normalized Index In Partition Includes One=true
    Number Of Partitions=$System.Columns
    Particles Per Partition:NiagaraInt32=$System.Rows
    Partition Mode:ENiagara_PartitionMode_v2=::NewEnumerator1
   )
   &InitializeParticle(
    Color Mode=::NewEnumerator0
    Mesh Scale=&Multiply_VectorByFloat(Vector=<0.1 0.03 0.02>, Float=0.28)
    Mesh Scale Mode=::NewEnumerator4
    Position Offset=&Add_Vector(
     A=<8.6 -2.5 -9.75>
     B=&Add_Vector(
      A=&Lerp_Vector(
       A=<0 0 -11.15>
       B=<0 0 11.15>
       Alpha=&Multiply_Float(A=$Particles.PartitionParticles.NormalizedPartition, B=1.1667)
      )
      B=&Lerp_Vector(A=<0 0 0>, B=<15 0 0>, Alpha=$Particles.PartitionParticles.Normalized Index In Partition)
     )
    )
    Sprite Rotation Angle=90
    Sprite Rotation Mode=::NewEnumerator0
    Sprite Size Mode=::NewEnumerator3
    Uniform Sprite Size=5
    UsePositionOffset=true
    Write Material Random=true
   )
  ]
  particleUpdateScript [
   &ParticleState(Kill Particles When Lifetime Has Elapsed=false)
   /SetVariables(Transient.NoiseCurve:NiagaraFloat=&Waveform(
     Waveform Count=::NewEnumerator1
     "Waveform [1]"=::NewEnumerator0
     "[1] Phase Scale"=&Multiply_FloatByInt(Float=0.2, Integer=$Particles.PartitionParticles.Partition)
     "[1] Frequency"=1.7
     "[1] Amplitude Min/Max"=<0 1>
     "[2] Average"=false
     "[2] Input Position"=0
     "[2] Phase"=$Emitter.Age
     "[2] Phase Scale"=&Multiply_FloatByInt(
      Float=0.2
      Integer=&Subtract_Integer(A:NiagaraInt32=$System.Columns, B:NiagaraInt32=$Particles.PartitionParticles.Partition)
     )
     "[2] Compound Frequency"=2
     "[2] Frequency"=1.5
     "[2] Amplitude Min/Max"=<0 1>
     "[2] Scale"=1
     "[2] Bias"=0
     "[2] Exponent"=1
     "[2] Pulse Duration"=0.1
     "[2] Clip Triangle Min/Max"=<0 0>
     "[2] Random Seed"=0
    )) {isSetParameters=true, module=SetVariables_#8}
   /SetVariables(Transient.EQCurve:LinearColor=&ColorFromCurve(
     ColorCurve=DataInterface() {propertyValues {
       AlphaCurve {keys [<at 0 1 RCIM_Cubic>]}
       BlueCurve {keys [<at 0 1>,<at 0.3456 0.7629 RCIM_Cubic>,<at 0.8091 1 RCIM_Cubic>]}
       GreenCurve {keys [<at 0 0.6344>, {arriveTangent=-0.784, interpMode=RCIM_Cubic, leaveTangent=-0.784, time=0.3456, value=1}, <at 0.8091 0 RCIM_Cubic>]}
       RedCurve {keys [<at 0 0.04>, {arriveTangent=0.7587, interpMode=RCIM_Cubic, leaveTangent=0.7587, time=0.3456, value=0.3217}, <at 0.8091 0.6539 RCIM_Cubic>]}
      }}
     CurveIndex=$Particles.PartitionParticles.Normalized Index In Partition
    )) {isSetParameters=true, module=SetVariables_#9}
   /SetVariables(Transient.EQCurve:LinearColor=&ColorFromCurve(
     ColorCurve=DataInterface() {propertyValues {
       AlphaCurve {keys [<at 0 1 RCIM_Cubic>]}
       BlueCurve {keys [
         <at 0 1>
         {arriveTangent=-0.2516, interpMode=RCIM_Cubic, leaveTangent=-0.2516, time=0.5713, value=0.7294}
         <at 0.592 0.851>
         <at 0.8091 1 RCIM_Cubic>
        ]}
       GreenCurve {keys [
         <at 0 0.327>
         {arriveTangent=0.8852, interpMode=RCIM_Cubic, leaveTangent=0.8852, time=0.5713, value=1}
         <at 0.592 0.851>
         <at 0.8091 0 RCIM_Cubic>
        ]}
       RedCurve {keys [
         <at 0 0>
         {arriveTangent=1.6891, interpMode=RCIM_Cubic, leaveTangent=1.6891, time=0.5713, value=0.226}
         <at 0.592 1>
         <at 0.8091 0.6539 RCIM_Cubic>
        ]}
      }}
     CurveIndex=$Particles.PartitionParticles.Normalized Index In Partition
    )) {enabled=false, isSetParameters=true, module=SetVariables_#10}
   &self:MultsFromSpectrum(
    Column Selector:NiagaraInt32=$Particles.PartitionParticles.Partition
    Modifier:NiagaraFloat=&Lerp_Float(
     A=$NPC.NPC_Audio_Reactive_Parameters.Amplitude_Music
     B=$Transient.NoiseCurve
     Alpha=&MakeCustomFloatFromBool(Bool=$System.bIsPreview, True Float=1, False Float=0.25)
    )
    Power Spectrum=3
    Row Selector:NiagaraFloat=$Particles.PartitionParticles.Normalized Index In Partition
    Spectrum Param:LinearColor=&LerpLinearColors(
     StartColor=$NPC.NPC_Audio_Reactive_Parameters.Spectrum_Music
     EndColor=$System.FakeSpectrum
     LerpFactor=&MakeCustomFloatFromBool(Bool=$System.bIsPreview, True Float=1, False Float=0)
    )
    Use Normalized=true
   ) {module=MultsFromSpectrum}
   &ScaleColor(
    Scale Alpha=&FloatFromCurve(
     FloatCurve=DataInterface() {propertyValues {Curve {keys [<at 0 1>,<at 1 0.5>]}}}
     CurveIndex=$Particles.PartitionParticles.Normalized Index In Partition
    )
    Scale RGB=&MakeVectorFromLinearColorRGB(Color=$Transient.EQCurve)
    ScaleA=true
    ScaleRGB=true
   )
   &ScaleMeshSize(Scale Factor=&VectorFromFloat(Value=$Output.MultsFromSpectrum.SoftMult))
   &ScaleSpriteSize(
    Scale Sprite Size Mode=::NewEnumerator0
    Uniform Scale Factor=&MakeCustomFloatFromBool(
     Bool=&SetBoolByFloatComparison(A=$Output.MultsFromSpectrum.NormalizedMult, B=$Particles.PartitionParticles.Normalized Index In Partition, Comparison Type=::NewEnumerator6)
     True Float=0
     False Float=1
    )
   )
   /SetVariables(
    Particles.SpriteColor=&MakeLinearColorFromVectorAndFloat(
     "Vector (RGB)"=&MakeVectorFromLinearColorRGB(Color=$Particles.Color)
     "Float (Alpha)"=&FloatFromCurve(
      FloatCurve=DataInterface() {propertyValues {Curve {keys [<at 0 0.2>,<at 1 0.4>]}}}
      CurveIndex=$Particles.PartitionParticles.Normalized Index In Partition
      Scale Curve=&MakeCustomFloatFromBool(
       Bool=&SetBoolByFloatComparison(A=$Output.MultsFromSpectrum.NormalizedMult, B=$Particles.PartitionParticles.Normalized Index In Partition, Comparison Type=::NewEnumerator6)
       True Float=0
       False Float=1
      )
     )
    )
    Particles.SpritePosition:NiagaraPosition=&AddVectorToPosition(Position=$Particles.Position, Vector=<1 0 0>)
   ) {isSetParameters=true, module=SetVariables_#11}
  ]
  rendererClasses [&NiagaraSpriteRendererProperties, &NiagaraMeshRendererProperties]
  renderers [
   {rendererClass=&NiagaraSpriteRendererProperties}
   {rendererClass=&NiagaraMeshRendererProperties, rendererIndex=1}
  ]
 }
 Values {
  bEnabled=false
  emitterUpdateScript [
   &EmitterState()
   &SpawnBurst_Instantaneous(Spawn Count=$System.Columns, Use Spawn Probability=false, UseLoopCountLimit=true)
  ]
  particleSpawnScript [
   &PartitionParticles(
    Normalized Index In Partition Includes One=true
    Number Of Partitions=$System.Columns
    Particles Per Partition=1
    Partition Mode:ENiagara_PartitionMode_v2=::NewEnumerator1
   )
   &InitializeParticle(
    Color Mode=::NewEnumerator0
    Mesh Scale=&Multiply_VectorByFloat(Vector=<0.03 0.03 0.1>, Float=0.28)
    Mesh Scale Mode=::NewEnumerator4
    Position Offset=&Add_Vector(
     A=<26 -2.5 -9.75>
     B=&Lerp_Vector(
      A=<0 0 -11.15>
      B=<0 0 11.15>
      Alpha=&Multiply_Float(A=$Particles.PartitionParticles.NormalizedPartition, B=1.1667)
     )
    )
    Sprite Size Mode=::NewEnumerator3
    UsePositionOffset=true
    Write Material Random=true
   )
  ]
  particleUpdateScript [
   &ParticleState(Kill Particles When Lifetime Has Elapsed=false)
   &self:MultsFromSpectrum(Column Selector:NiagaraInt32=$Particles.PartitionParticles.Partition, Spectrum Param:LinearColor=$System.ChooseSpectrum) {module=MultsFromSpectrum}
   /SetVariables(Transient.NoiseCurve:NiagaraFloat=&MakeCustomFloatFromBool(
     Bool=$System.bIsPreview
     True Float=&Sine(Normalized Angle=$Emitter.Age, Period=0.6, Scale=0.25, Bias=0.5)
     False Float=$NPC.NPC_Audio_Reactive_Parameters.BeatGameTime
    )) {isSetParameters=true}
   /SetVariables(Transient.AlphaMask:NiagaraFloat=&Add_Float(
     A=$Transient.NoiseCurve
     B=&Multiply_Float(A=0.5, B=&Power(a=$Output.MultsFromSpectrum.BaseMult, b=2))
    )) {isSetParameters=true}
   &DynamicMaterialParameters(
    Index 0 Param 1=$NPC.NPC_Audio_Reactive_Parameters.Amplitude_Music
    Module.Index0Float0Write=true
    Module.Index0Float1Write=true
    Module.Index0Float2Write=true
    Module.Index0Float3Write=true
   )
  ]
  rendererClasses [&NiagaraSpriteRendererProperties]
  renderers [{rendererClass=&NiagaraSpriteRendererProperties}]
 }
}
summary {
 systemName=NS_Keytar_FareSpore_Idle_Visualizer_v1
 emitters [
  {emitterName=Chips, rendererClasses [&NiagaraSpriteRendererProperties, &NiagaraMeshRendererProperties]}
  {bEnabled=false, emitterName=Values, rendererClasses [&NiagaraSpriteRendererProperties]}
  {
   bEnabled=false
   emitterName=Chips001
   rendererClasses [&NiagaraSpriteRendererProperties, &NiagaraMeshRendererProperties]
  }
  {
   bEnabled=false
   emitterName=ChipsNorm
   rendererClasses [&NiagaraSpriteRendererProperties, &NiagaraMeshRendererProperties]
  }
 ]
 userVariables [User.bIsFrontEndPreview {type=NiagaraBool}]
}
