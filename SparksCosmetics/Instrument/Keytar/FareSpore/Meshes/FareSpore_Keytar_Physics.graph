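# Physics asset for the FareSpore keytar cosmetic, simulating a six-bone dangling chain attached to the instrument's handle.
# Small spheres form the links, and the tip combines a larger head sphere with two side spheres.
# A single rotated box covers the handle.
# Angular constraints join each link to its predecessor, with twist locked at the handle joint and swing limits generally widening from the second link toward the tip.
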PhysicsAsset /SparksCosmetics/Instrument/Keytar/FareSpore/Meshes/FareSpore_Keytar_Physics
bodies {
 dyn_chain_01 [{shapeType=Sphere, radius=0.5, shapeName=dyn_chain_01_sphere}]
 dyn_chain_02 [{shapeType=Sphere, radius=0.5, shapeName=dyn_chain_02_sphere, center=<0.4 0 0>}]
 dyn_chain_03 [{shapeType=Sphere, radius=0.5, shapeName=dyn_chain_03_sphere, center=<0.4 0 0>}]
 dyn_chain_04 [{shapeType=Sphere, radius=0.5, shapeName=dyn_chain_04_sphere, center=<0.4 0 0>}]
 dyn_chain_05 [{shapeType=Sphere, radius=0.5, shapeName=dyn_chain_05_sphere, center=<0.4 0 0>}]
 dyn_chain_06 [
  {shapeType=Sphere, radius=2.0385, shapeName=dyn_chain_06_sphere_head, center=<2 0 0>}
  {shapeType=Sphere, radius=1.2602, shapeName=dyn_chain_06_sphere, center=<0.8894 1.9531 0.4107>}
  {shapeType=Sphere, radius=1.2602, shapeName=dyn_chain_06_sphere, center=<0.782 -1.9267 0.4107>}
 ]
 handle [{
   shapeType=Box
   extentX=16.8576
   extentY=2.7418
   extentZ=11.5137
   shapeName=handle_box
   center=<1.1957 -7.2296 -17.1366>
   rotation=<rot 0 -90 0>
  }]
}
constraints [
 {bone1Name=dyn_chain_06, bone2Name=dyn_chain_05, swing1LimitDegrees=35, swing2LimitDegrees=50, twistLimitDegrees=10}
 {bone1Name=dyn_chain_01, bone2Name=handle, swing1LimitDegrees=30, swing2LimitDegrees=45, twistLimitDegrees=45, twistMotion=Locked}
 {bone1Name=dyn_chain_02, bone2Name=dyn_chain_01, swing1LimitDegrees=20, swing2LimitDegrees=45, twistLimitDegrees=8}
 {bone1Name=dyn_chain_03, bone2Name=dyn_chain_02, swing1LimitDegrees=25, swing2LimitDegrees=50, twistLimitDegrees=10}
 {bone1Name=dyn_chain_04, bone2Name=dyn_chain_03, swing1LimitDegrees=30, swing2LimitDegrees=50, twistLimitDegrees=10}
 {bone1Name=dyn_chain_05, bone2Name=dyn_chain_04, swing1LimitDegrees=30, swing2LimitDegrees=50, twistLimitDegrees=10}
]
